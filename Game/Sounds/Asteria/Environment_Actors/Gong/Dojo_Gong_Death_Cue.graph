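# Sound cue for the death of the Dojo gong environment actor, layering a gong death sound with a wood-break crack.
# A mixer plays a random gong death wave with slight upward pitch variation, louder than a second, briefly delayed layer.
# That layer picks one of four borrowed Husk Pitcher torch wood-break impacts and pitches it down.
# It uses chime impact attenuation, the impacts sound class and a gong concurrency set.
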SoundCue /Game/Sounds/Asteria/Environment_Actors/Gong/Dojo_Gong_Death_Cue
root=Mixer_1
nodes [
 {id=Delay_0, children [Modulator_1], properties {delayMax=0.06, delayMin=0.05}}
 {id=Mixer_1, children [Modulator_0,Delay_0], properties {inputVolume [1.5,0.6]}}
 {id=Modulator_0, children [Random_2], properties {pitchMax=1.1, pitchMin=1}}
 {id=Modulator_1, children [Random_1], properties {pitchMax=0.85, pitchMin=0.7}}
 {
  id=Random_1
  children [WavePlayer_9, WavePlayer_8, WavePlayer_10, WavePlayer_11]
  properties {bShouldExcludeFromBranchCulling=true, weights [1,1,1,1]}
 }
 {id=Random_2, children [WavePlayer_7, WavePlayer_12], properties {weights [1,1]}}
 {
  id=WavePlayer_10
  properties {soundWaveAssetPtr=/Game/Sounds/Fort_Monster_Husk_Pitcher_Villager/Proj_Impact_Villager_Husk_Pitcher_Torch_Wood_Break_03_}
 }
 {
  id=WavePlayer_11
  properties {soundWaveAssetPtr=/Game/Sounds/Fort_Monster_Husk_Pitcher_Villager/Proj_Impact_Villager_Husk_Pitcher_Torch_Wood_Break_04_}
 }
 {id=WavePlayer_12, properties {soundWaveAssetPtr=&/Dojo_Gong_Death_01}}
 {id=WavePlayer_7, properties {soundWaveAssetPtr=&/Dojo_Gong_Death_02}}
 {
  id=WavePlayer_8
  properties {soundWaveAssetPtr=/Game/Sounds/Fort_Monster_Husk_Pitcher_Villager/Proj_Impact_Villager_Husk_Pitcher_Torch_Wood_Break_02_}
 }
 {
  id=WavePlayer_9
  properties {soundWaveAssetPtr=/Game/Sounds/Fort_Monster_Husk_Pitcher_Villager/Proj_Impact_Villager_Husk_Pitcher_Torch_Wood_Break_01_}
 }
]
properties {
 attenuationSettings=/Game/Sounds/Fort_Ambient_Looping/CedarForestChime/CedarWoods_Chime_Impact_Att
 maxDistance=4800
 priority=2
 soundClassObject=&Fort_Impacts
 volumeMultiplier=1.2
 concurrencySet [&/Dojo_Gong_Conc]
}
