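# Sound cue for a dojo gong impact environment sound, classed under Fort_Impacts.
# A Random node picks one of four equally weighted impact waves and is excluded from branch culling.
# A Modulator above it varies pitch and boosts volume within a random range, on top of a raised cue volume multiplier.
# It reuses the CedarWoods chime impact attenuation and references a dedicated gong concurrency asset.
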SoundCue /Game/Sounds/Asteria/Environment_Actors/Gong/Dojo_Gong_Impact_Cue
root=Modulator_0
nodes [
 {id=Modulator_0, children [Random_1], properties {pitchMax=1, volumeMax=1.5, volumeMin=1.3}}
 {
  id=Random_1
  children [WavePlayer_7, WavePlayer_8, WavePlayer_9, WavePlayer_10]
  properties {bShouldExcludeFromBranchCulling=true, weights [1,1,1,1]}
 }
 {id=WavePlayer_10, properties {soundWaveAssetPtr=&/Dojo_Gong_Impact_04}}
 {id=WavePlayer_7, properties {soundWaveAssetPtr=&/Dojo_Gong_Impact_01}}
 {id=WavePlayer_8, properties {soundWaveAssetPtr=&/Dojo_Gong_Impact_02}}
 {id=WavePlayer_9, properties {soundWaveAssetPtr=&/Dojo_Gong_Impact_03}}
]
properties {
 attenuationSettings=/Game/Sounds/Fort_Ambient_Looping/CedarForestChime/CedarWoods_Chime_Impact_Att
 maxDistance=4800
 soundClassObject=&Fort_Impacts
 volumeMultiplier=4
 concurrencyOverrides {bLimitToOwner=true, maxCount=1, resolutionRule=StopOldest, volumeScaleAttackTime=1}
 concurrencySet [&/Dojo_Gong_Conc]
}
